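# Animation Blueprint for the SpruceBark backpack, playing idle, skydiving and reactions to damage, healing, eliminations and chest opening.
# A gameplay or front-end state machine feeds additive bone offsets, including a sine-driven bob, into a cached pose.
# Two matching AnimDynamics chains simulate beard and feathers from that pose, selected by bIsInFrontEnd.
# The event graph detects health changes, times front-end reactions and derives physics weights from the IsPhysicsDisabled curve.
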AnimBlueprint /BRCosmetics/Accessories/FORT_Backpacks/Backpack_M_MED_SpruceBark/Meshes/M_MED_SpruceBark_Pack_AnimBP
target_skeleton=&/M_MED_SpruceBark_Pack_Skeleton
graphs [
 AnimGraph {nodes [
   Input Pose()
   n1 = Anim Dynamics - Bone: Beard(ComponentPose=Local To Component(LocalPose=Use cached pose 'PreDyn'()), Alpha=1, ExternalForce="0,0,0") {
    angularDampingOverride=0.999
    bOverrideAngularDamping=true
    simulationSpace=RootRelative
    boundBone {boneName=Beard}
    componentAppliedLinearAccClamp=<3000 3000 3000>
    componentLinearAccScale=<0.3 0.3 0.3>
    componentLinearVelScale=<0.3 0.3 0.3>
    physicsBodyDefinitions [{
      boundBone {boneName=Beard}
      boxExtents=<10 10 10>
      constraintSetup {angularLimitsMax=<15 15 30>, angularLimitsMin=<-15 -15 -30>, angularTarget=<1 0 0>}
      localJointOffset=<1 0 0>
     }]
    simSpaceSettings {maxAngularAcceleration=25, maxAngularVelocity=5, simSpaceAngularAlpha=1}
   }
   n2 = Anim Dynamics - Bone: Feather_01_1(ComponentPose=@n1, Alpha=1, ExternalForce="0,0,0") {
    angularDampingOverride=1
    bGravityOverrideInSimSpace=true
    bOverrideAngularDamping=true
    bUseGravityOverride=true
    simulationSpace=RootRelative
    alphaScaleBiasClamp {bMapRange=true}
    boundBone {boneName=Feather_01_1}
    componentAppliedLinearAccClamp=<5000 5000 5000>
    componentLinearAccScale=<0.5 0.5 0.5>
    componentLinearVelScale=<0.5 0.5 0.5>
    gravityOverride {z=980}
    physicsBodyDefinitions [{
      boundBone {boneName=Feather_01_1}
      boxExtents=<10 10 10>
      constraintSetup {angularLimitsMax=<5 5 0>, angularLimitsMin=<-5 -5 -5>, angularTarget=<1 0 0>}
      localJointOffset=<0.75 0 0>
     }]
    simSpaceSettings {maxAngularAcceleration=25, maxAngularVelocity=5, simSpaceAngularAlpha=1}
   }
   n3 = Anim Dynamics - Bone: Feather_01_2(ComponentPose=@n2, Alpha=1, ExternalForce="0,0,0") {
    angularDampingOverride=1
    bGravityOverrideInSimSpace=true
    bOverrideAngularDamping=true
    bUseGravityOverride=true
    simulationSpace=RootRelative
    boundBone {boneName=Feather_01_2}
    componentAppliedLinearAccClamp=<5000 5000 5000>
    componentLinearAccScale=<0.5 0.5 0.5>
    componentLinearVelScale=<0.5 0.5 0.5>
    gravityOverride {z=980}
    physicsBodyDefinitions [{
      boundBone {boneName=Feather_01_2}
      boxExtents=<10 10 10>
      constraintSetup {angularLimitsMax=<10 15 0>, angularLimitsMin=<-10 -15 -15>, angularTarget=<1 0 0>}
      localJointOffset=<0.5 0 0>
     }]
    simSpaceSettings {maxAngularAcceleration=25, maxAngularVelocity=5, simSpaceAngularAlpha=1}
   }
   n4 = Anim Dynamics - Bone: Feather_01_3(ComponentPose=@n3, Alpha=1, ExternalForce="0,0,0") {
    angularDampingOverride=1
    bGravityOverrideInSimSpace=true
    bOverrideAngularDamping=true
    bUseGravityOverride=true
    simulationSpace=RootRelative
    boundBone {boneName=Feather_01_3}
    componentAppliedLinearAccClamp=<3000 3000 3000>
    componentLinearAccScale=<0.5 0.5 0.5>
    componentLinearVelScale=<0.5 0.5 0.5>
    gravityOverride {z=980}
    physicsBodyDefinitions [{
      boundBone {boneName=Feather_01_3}
      boxExtents=<10 10 10>
      constraintSetup {angularLimitsMax=<10 15 0>, angularLimitsMin=<-10 -15 -15>, angularTarget=<1 0 0>}
      localJointOffset=<0.5 0 0>
     }]
    simSpaceSettings {maxAngularAcceleration=25, maxAngularVelocity=5, simSpaceAngularAlpha=1}
   }
   n5 = Anim Dynamics - Bone: Feather_01_4(ComponentPose=@n4, Alpha=1, ExternalForce="0,0,0") {
    angularDampingOverride=0.999
    bGravityOverrideInSimSpace=true
    bOverrideAngularDamping=true
    bUseGravityOverride=true
    simulationSpace=RootRelative
    boundBone {boneName=Feather_01_4}
    componentAppliedLinearAccClamp=<3000 3000 3000>
    componentLinearAccScale=<0.5 0.5 0.5>
    componentLinearVelScale=<0.5 0.5 0.5>
    gravityOverride {z=980}
    physicsBodyDefinitions [{
      boundBone {boneName=Feather_01_4}
      boxExtents=<10 10 10>
      constraintSetup {angularLimitsMax=<10 15 10>, angularLimitsMin=<-10 -15 -10>, angularTarget=<1 0 0>}
      localJointOffset=<0.75 0 0>
     }]
    simSpaceSettings {maxAngularAcceleration=25, maxAngularVelocity=5, simSpaceAngularAlpha=1}
   }
   n6 = Anim Dynamics - Bone: Feather_02_1(ComponentPose=@n5, Alpha=1, ExternalForce="0,0,0") {
    angularDampingOverride=1
    bGravityOverrideInSimSpace=true
    bOverrideAngularDamping=true
    bUseGravityOverride=true
    simulationSpace=RootRelative
    alphaScaleBiasClamp {bMapRange=true}
    boundBone {boneName=Feather_02_1}
    componentAppliedLinearAccClamp=<5000 5000 5000>
    componentLinearAccScale=<0.5 0.5 0.5>
    componentLinearVelScale=<0.5 0.5 0.5>
    gravityOverride {z=980}
    physicsBodyDefinitions [{
      boundBone {boneName=Feather_02_1}
      boxExtents=<10 10 10>
      constraintSetup {angularLimitsMax=<5 5 0>, angularLimitsMin=<-5 -5 -5>, angularTarget=<1 0 0>}
      localJointOffset=<0.75 0 0>
     }]
    simSpaceSettings {maxAngularAcceleration=25, maxAngularVelocity=5, simSpaceAngularAlpha=1}
   }
   n7 = Anim Dynamics - Bone: Feather_02_2(ComponentPose=@n6, Alpha=1, ExternalForce="0,0,0") {
    angularDampingOverride=1
    bGravityOverrideInSimSpace=true
    bOverrideAngularDamping=true
    bUseGravityOverride=true
    simulationSpace=RootRelative
    boundBone {boneName=Feather_02_2}
    componentAppliedLinearAccClamp=<5000 5000 5000>
    componentLinearAccScale=<0.5 0.5 0.5>
    componentLinearVelScale=<0.5 0.5 0.5>
    gravityOverride {z=980}
    physicsBodyDefinitions [{
      boundBone {boneName=Feather_02_2}
      boxExtents=<10 10 10>
      constraintSetup {angularLimitsMax=<15 15 0>, angularLimitsMin=<-3 -15 -15>, angularTarget=<1 0 0>}
      localJointOffset=<0.5 0 0>
     }]
    simSpaceSettings {maxAngularAcceleration=25, maxAngularVelocity=5, simSpaceAngularAlpha=1}
   }
   n8 = Anim Dynamics - Bone: Feather_02_3(ComponentPose=@n7, Alpha=1, ExternalForce="0,0,0") {
    angularDampingOverride=1
    bGravityOverrideInSimSpace=true
    bOverrideAngularDamping=true
    bUseGravityOverride=true
    simulationSpace=RootRelative
    boundBone {boneName=Feather_02_3}
    componentAppliedLinearAccClamp=<3000 3000 3000>
    componentLinearAccScale=<0.5 0.5 0.5>
    componentLinearVelScale=<0.5 0.5 0.5>
    gravityOverride {z=980}
    physicsBodyDefinitions [{
      boundBone {boneName=Feather_02_3}
      boxExtents=<10 10 10>
      constraintSetup {angularLimitsMax=<15 15 0>, angularLimitsMin=<-3 -15 -15>, angularTarget=<1 0 0>}
      localJointOffset=<0.5 0 0>
     }]
    simSpaceSettings {maxAngularAcceleration=25, maxAngularVelocity=5, simSpaceAngularAlpha=1}
   }
   n9 = Anim Dynamics - Bone: Feather_02_4(ComponentPose=@n8, Alpha=1, ExternalForce="0,0,0") {
    angularDampingOverride=0.999
    bGravityOverrideInSimSpace=true
    bOverrideAngularDamping=true
    bUseGravityOverride=true
    simulationSpace=RootRelative
    boundBone {boneName=Feather_02_4}
    componentAppliedLinearAccClamp=<3000 3000 3000>
    componentLinearAccScale=<0.5 0.5 0.5>
    componentLinearVelScale=<0.5 0.5 0.5>
    gravityOverride {z=980}
    physicsBodyDefinitions [{
      boundBone {boneName=Feather_02_4}
      boxExtents=<10 10 10>
      constraintSetup {angularLimitsMax=<15 15 10>, angularLimitsMin=<-3 -15 -10>, angularTarget=<1 0 0>}
      localJointOffset=<0.75 0 0>
     }]
    simSpaceSettings {maxAngularAcceleration=25, maxAngularVelocity=5, simSpaceAngularAlpha=1}
   }
   n10 = Anim Dynamics - Bone: Feather_03_1(ComponentPose=@n9, Alpha=1, ExternalForce="0,0,0") {
    angularDampingOverride=1
    bGravityOverrideInSimSpace=true
    bOverrideAngularDamping=true
    bUseGravityOverride=true
    simulationSpace=RootRelative
    alphaScaleBiasClamp {bMapRange=true}
    boundBone {boneName=Feather_03_1}
    componentAppliedLinearAccClamp=<5000 5000 5000>
    componentLinearAccScale=<0.5 0.5 0.5>
    componentLinearVelScale=<0.5 0.5 0.5>
    gravityOverride {z=980}
    physicsBodyDefinitions [{
      boundBone {boneName=Feather_03_1}
      boxExtents=<10 10 10>
      constraintSetup {angularLimitsMax=<5 5 0>, angularLimitsMin=<-5 -5 -5>, angularTarget=<1 0 0>}
      localJointOffset=<0.75 0 0>
     }]
    simSpaceSettings {maxAngularAcceleration=25, maxAngularVelocity=5, simSpaceAngularAlpha=1}
   }
   n11 = Anim Dynamics - Bone: Feather_03_2(ComponentPose=@n10, Alpha=1, ExternalForce="0,0,0") {
    angularDampingOverride=1
    bGravityOverrideInSimSpace=true
    bOverrideAngularDamping=true
    bUseGravityOverride=true
    simulationSpace=RootRelative
    boundBone {boneName=Feather_03_2}
    componentAppliedLinearAccClamp=<5000 5000 5000>
    componentLinearAccScale=<0.5 0.5 0.5>
    componentLinearVelScale=<0.5 0.5 0.5>
    gravityOverride {z=980}
    physicsBodyDefinitions [{
      boundBone {boneName=Feather_03_2}
      boxExtents=<10 10 10>
      constraintSetup {angularLimitsMax=<3 15 0>, angularLimitsMin=<-15 -15 -15>, angularTarget=<1 0 0>}
      localJointOffset=<0.5 0 0>
     }]
    simSpaceSettings {maxAngularAcceleration=25, maxAngularVelocity=5, simSpaceAngularAlpha=1}
   }
   n12 = Anim Dynamics - Bone: Feather_03_3(ComponentPose=@n11, Alpha=1, ExternalForce="0,0,0") {
    angularDampingOverride=1
    bGravityOverrideInSimSpace=true
    bOverrideAngularDamping=true
    bUseGravityOverride=true
    simulationSpace=RootRelative
    boundBone {boneName=Feather_03_3}
    componentAppliedLinearAccClamp=<3000 3000 3000>
    componentLinearAccScale=<0.5 0.5 0.5>
    componentLinearVelScale=<0.5 0.5 0.5>
    gravityOverride {z=980}
    physicsBodyDefinitions [{
      boundBone {boneName=Feather_03_3}
      boxExtents=<10 10 10>
      constraintSetup {angularLimitsMax=<3 15 0>, angularLimitsMin=<-15 -15 -15>, angularTarget=<1 0 0>}
      localJointOffset=<0.5 0 0>
     }]
    simSpaceSettings {maxAngularAcceleration=25, maxAngularVelocity=5, simSpaceAngularAlpha=1}
   }
   n13 = Anim Dynamics - Bone: Feather_03_4(ComponentPose=@n12, Alpha=1, ExternalForce="0,0,0") {
    angularDampingOverride=0.999
    bGravityOverrideInSimSpace=true
    bOverrideAngularDamping=true
    bUseGravityOverride=true
    simulationSpace=RootRelative
    boundBone {boneName=Feather_03_4}
    componentAppliedLinearAccClamp=<3000 3000 3000>
    componentLinearAccScale=<0.5 0.5 0.5>
    componentLinearVelScale=<0.5 0.5 0.5>
    gravityOverride {z=980}
    physicsBodyDefinitions [{
      boundBone {boneName=Feather_03_4}
      boxExtents=<10 10 10>
      constraintSetup {angularLimitsMax=<3 15 10>, angularLimitsMin=<-15 -15 -10>, angularTarget=<1 0 0>}
      localJointOffset=<0.75 0 0>
     }]
    simSpaceSettings {maxAngularAcceleration=25, maxAngularVelocity=5, simSpaceAngularAlpha=1}
   }
   n14 = Anim Dynamics - Bone: Feather_04_1(ComponentPose=@n13, Alpha=1, ExternalForce="0,0,0") {
    angularDampingOverride=1
    bGravityOverrideInSimSpace=true
    bOverrideAngularDamping=true
    bUseGravityOverride=true
    simulationSpace=RootRelative
    alphaScaleBiasClamp {bMapRange=true}
    boundBone {boneName=Feather_04_1}
    componentAppliedLinearAccClamp=<5000 5000 5000>
    componentLinearAccScale=<0.5 0.5 0.5>
    componentLinearVelScale=<0.5 0.5 0.5>
    gravityOverride {z=980}
    physicsBodyDefinitions [{
      boundBone {boneName=Feather_04_1}
      boxExtents=<10 10 10>
      constraintSetup {angularLimitsMax=<5 5 0>, angularLimitsMin=<-5 -5 -5>, angularTarget=<1 0 0>}
      localJointOffset=<0.75 0 0>
     }]
    simSpaceSettings {maxAngularAcceleration=25, maxAngularVelocity=5, simSpaceAngularAlpha=1}
   }
   n15 = Anim Dynamics - Bone: Feather_04_2(ComponentPose=@n14, Alpha=1, ExternalForce="0,0,0") {
    angularDampingOverride=1
    bGravityOverrideInSimSpace=true
    bOverrideAngularDamping=true
    bUseGravityOverride=true
    simulationSpace=RootRelative
    boundBone {boneName=Feather_04_2}
    componentAppliedLinearAccClamp=<5000 5000 5000>
    componentLinearAccScale=<0.5 0.5 0.5>
    componentLinearVelScale=<0.5 0.5 0.5>
    gravityOverride {z=980}
    physicsBodyDefinitions [{
      boundBone {boneName=Feather_04_2}
      boxExtents=<10 10 10>
      constraintSetup {angularLimitsMax=<10 15 0>, angularLimitsMin=<-10 -15 -15>, angularTarget=<1 0 0>}
      localJointOffset=<0.5 0 0>
     }]
    simSpaceSettings {maxAngularAcceleration=25, maxAngularVelocity=5, simSpaceAngularAlpha=1}
   }
   n16 = Anim Dynamics - Bone: Feather_04_3(ComponentPose=@n15, Alpha=1, ExternalForce="0,0,0") {
    angularDampingOverride=1
    bGravityOverrideInSimSpace=true
    bOverrideAngularDamping=true
    bUseGravityOverride=true
    simulationSpace=RootRelative
    boundBone {boneName=Feather_04_3}
    componentAppliedLinearAccClamp=<3000 3000 3000>
    componentLinearAccScale=<0.5 0.5 0.5>
    componentLinearVelScale=<0.5 0.5 0.5>
    gravityOverride {z=980}
    physicsBodyDefinitions [{
      boundBone {boneName=Feather_04_3}
      boxExtents=<10 10 10>
      constraintSetup {angularLimitsMax=<10 15 0>, angularLimitsMin=<-10 -15 -15>, angularTarget=<1 0 0>}
      localJointOffset=<0.5 0 0>
     }]
    simSpaceSettings {maxAngularAcceleration=25, maxAngularVelocity=5, simSpaceAngularAlpha=1}
   }
   n17 = Anim Dynamics - Bone: Beard(ComponentPose=Local To Component(LocalPose=Use cached pose 'PreDyn'()), Alpha=1, ExternalForce="0,0,0") {
    angularDampingOverride=0.999
    bOverrideAngularDamping=true
    simulationSpace=RootRelative
    boundBone {boneName=Beard}
    componentAppliedLinearAccClamp=<3000 3000 3000>
    componentLinearAccScale=<0.3 0.3 0.3>
    componentLinearVelScale=<0.3 0.3 0.3>
    physicsBodyDefinitions [{
      boundBone {boneName=Beard}
      boxExtents=<10 10 10>
      constraintSetup {angularLimitsMax=<15 15 30>, angularLimitsMin=<-15 -15 -30>, angularTarget=<1 0 0>}
      localJointOffset=<1 0 0>
     }]
    simSpaceSettings {maxAngularAcceleration=25, maxAngularVelocity=5, simSpaceAngularAlpha=1}
   }
   n18 = Anim Dynamics - Bone: Feather_01_1(ComponentPose=@n17, Alpha=1, ExternalForce="0,0,0") {
    angularDampingOverride=1
    bGravityOverrideInSimSpace=true
    bOverrideAngularDamping=true
    bUseGravityOverride=true
    simulationSpace=RootRelative
    alphaScaleBiasClamp {bMapRange=true}
    boundBone {boneName=Feather_01_1}
    componentAppliedLinearAccClamp=<5000 5000 5000>
    componentLinearAccScale=<0.5 0.5 0.5>
    componentLinearVelScale=<0.5 0.5 0.5>
    gravityOverride {z=980}
    physicsBodyDefinitions [{
      boundBone {boneName=Feather_01_1}
      boxExtents=<10 10 10>
      constraintSetup {angularLimitsMax=<5 5 0>, angularLimitsMin=<-5 -5 -5>, angularTarget=<1 0 0>}
      localJointOffset=<0.75 0 0>
     }]
    simSpaceSettings {maxAngularAcceleration=25, maxAngularVelocity=5, simSpaceAngularAlpha=1}
   }
   n19 = Anim Dynamics - Bone: Feather_01_2(ComponentPose=@n18, Alpha=1, ExternalForce="0,0,0") {
    angularDampingOverride=1
    bGravityOverrideInSimSpace=true
    bOverrideAngularDamping=true
    bUseGravityOverride=true
    simulationSpace=RootRelative
    boundBone {boneName=Feather_01_2}
    componentAppliedLinearAccClamp=<5000 5000 5000>
    componentLinearAccScale=<0.5 0.5 0.5>
    componentLinearVelScale=<0.5 0.5 0.5>
    gravityOverride {z=980}
    physicsBodyDefinitions [{
      boundBone {boneName=Feather_01_2}
      boxExtents=<10 10 10>
      constraintSetup {angularLimitsMax=<10 15 0>, angularLimitsMin=<-10 -15 -15>, angularTarget=<1 0 0>}
      localJointOffset=<0.5 0 0>
     }]
    simSpaceSettings {maxAngularAcceleration=25, maxAngularVelocity=5, simSpaceAngularAlpha=1}
   }
   n20 = Anim Dynamics - Bone: Feather_01_3(ComponentPose=@n19, Alpha=1, ExternalForce="0,0,0") {
    angularDampingOverride=1
    bGravityOverrideInSimSpace=true
    bOverrideAngularDamping=true
    bUseGravityOverride=true
    simulationSpace=RootRelative
    boundBone {boneName=Feather_01_3}
    componentAppliedLinearAccClamp=<3000 3000 3000>
    componentLinearAccScale=<0.5 0.5 0.5>
    componentLinearVelScale=<0.5 0.5 0.5>
    gravityOverride {z=980}
    physicsBodyDefinitions [{
      boundBone {boneName=Feather_01_3}
      boxExtents=<10 10 10>
      constraintSetup {angularLimitsMax=<10 15 0>, angularLimitsMin=<-10 -15 -15>, angularTarget=<1 0 0>}
      localJointOffset=<0.5 0 0>
     }]
    simSpaceSettings {maxAngularAcceleration=25, maxAngularVelocity=5, simSpaceAngularAlpha=1}
   }
   n21 = Anim Dynamics - Bone: Feather_01_4(ComponentPose=@n20, Alpha=1, ExternalForce="0,0,0") {
    angularDampingOverride=0.999
    bGravityOverrideInSimSpace=true
    bOverrideAngularDamping=true
    bUseGravityOverride=true
    simulationSpace=RootRelative
    boundBone {boneName=Feather_01_4}
    componentAppliedLinearAccClamp=<3000 3000 3000>
    componentLinearAccScale=<0.5 0.5 0.5>
    componentLinearVelScale=<0.5 0.5 0.5>
    gravityOverride {z=980}
    physicsBodyDefinitions [{
      boundBone {boneName=Feather_01_4}
      boxExtents=<10 10 10>
      constraintSetup {angularLimitsMax=<10 15 10>, angularLimitsMin=<-10 -15 -10>, angularTarget=<1 0 0>}
      localJointOffset=<0.75 0 0>
     }]
    simSpaceSettings {maxAngularAcceleration=25, maxAngularVelocity=5, simSpaceAngularAlpha=1}
   }
   n22 = Anim Dynamics - Bone: Feather_02_1(ComponentPose=@n21, Alpha=1, ExternalForce="0,0,0") {
    angularDampingOverride=1
    bGravityOverrideInSimSpace=true
    bOverrideAngularDamping=true
    bUseGravityOverride=true
    simulationSpace=RootRelative
    alphaScaleBiasClamp {bMapRange=true}
    boundBone {boneName=Feather_02_1}
    componentAppliedLinearAccClamp=<5000 5000 5000>
    componentLinearAccScale=<0.5 0.5 0.5>
    componentLinearVelScale=<0.5 0.5 0.5>
    gravityOverride {z=980}
    physicsBodyDefinitions [{
      boundBone {boneName=Feather_02_1}
      boxExtents=<10 10 10>
      constraintSetup {angularLimitsMax=<5 5 0>, angularLimitsMin=<-5 -5 -5>, angularTarget=<1 0 0>}
      localJointOffset=<0.75 0 0>
     }]
    simSpaceSettings {maxAngularAcceleration=25, maxAngularVelocity=5, simSpaceAngularAlpha=1}
   }
   n23 = Anim Dynamics - Bone: Feather_02_2(ComponentPose=@n22, Alpha=1, ExternalForce="0,0,0") {
    angularDampingOverride=1
    bGravityOverrideInSimSpace=true
    bOverrideAngularDamping=true
    bUseGravityOverride=true
    simulationSpace=RootRelative
    boundBone {boneName=Feather_02_2}
    componentAppliedLinearAccClamp=<5000 5000 5000>
    componentLinearAccScale=<0.5 0.5 0.5>
    componentLinearVelScale=<0.5 0.5 0.5>
    gravityOverride {z=980}
    physicsBodyDefinitions [{
      boundBone {boneName=Feather_02_2}
      boxExtents=<10 10 10>
      constraintSetup {angularLimitsMax=<15 15 0>, angularLimitsMin=<-3 -15 -15>, angularTarget=<1 0 0>}
      localJointOffset=<0.5 0 0>
     }]
    simSpaceSettings {maxAngularAcceleration=25, maxAngularVelocity=5, simSpaceAngularAlpha=1}
   }
   n24 = Anim Dynamics - Bone: Feather_02_3(ComponentPose=@n23, Alpha=1, ExternalForce="0,0,0") {
    angularDampingOverride=1
    bGravityOverrideInSimSpace=true
    bOverrideAngularDamping=true
    bUseGravityOverride=true
    simulationSpace=RootRelative
    boundBone {boneName=Feather_02_3}
    componentAppliedLinearAccClamp=<3000 3000 3000>
    componentLinearAccScale=<0.5 0.5 0.5>
    componentLinearVelScale=<0.5 0.5 0.5>
    gravityOverride {z=980}
    physicsBodyDefinitions [{
      boundBone {boneName=Feather_02_3}
      boxExtents=<10 10 10>
      constraintSetup {angularLimitsMax=<15 15 0>, angularLimitsMin=<-3 -15 -15>, angularTarget=<1 0 0>}
      localJointOffset=<0.5 0 0>
     }]
    simSpaceSettings {maxAngularAcceleration=25, maxAngularVelocity=5, simSpaceAngularAlpha=1}
   }
   n25 = Anim Dynamics - Bone: Feather_02_4(ComponentPose=@n24, Alpha=1, ExternalForce="0,0,0") {
    angularDampingOverride=0.999
    bGravityOverrideInSimSpace=true
    bOverrideAngularDamping=true
    bUseGravityOverride=true
    simulationSpace=RootRelative
    boundBone {boneName=Feather_02_4}
    componentAppliedLinearAccClamp=<3000 3000 3000>
    componentLinearAccScale=<0.5 0.5 0.5>
    componentLinearVelScale=<0.5 0.5 0.5>
    gravityOverride {z=980}
    physicsBodyDefinitions [{
      boundBone {boneName=Feather_02_4}
      boxExtents=<10 10 10>
      constraintSetup {angularLimitsMax=<15 15 10>, angularLimitsMin=<-3 -15 -10>, angularTarget=<1 0 0>}
      localJointOffset=<0.75 0 0>
     }]
    simSpaceSettings {maxAngularAcceleration=25, maxAngularVelocity=5, simSpaceAngularAlpha=1}
   }
   n26 = Anim Dynamics - Bone: Feather_03_1(ComponentPose=@n25, Alpha=1, ExternalForce="0,0,0") {
    angularDampingOverride=1
    bGravityOverrideInSimSpace=true
    bOverrideAngularDamping=true
    bUseGravityOverride=true
    simulationSpace=RootRelative
    alphaScaleBiasClamp {bMapRange=true}
    boundBone {boneName=Feather_03_1}
    componentAppliedLinearAccClamp=<5000 5000 5000>
    componentLinearAccScale=<0.5 0.5 0.5>
    componentLinearVelScale=<0.5 0.5 0.5>
    gravityOverride {z=980}
    physicsBodyDefinitions [{
      boundBone {boneName=Feather_03_1}
      boxExtents=<10 10 10>
      constraintSetup {angularLimitsMax=<5 5 0>, angularLimitsMin=<-5 -5 -5>, angularTarget=<1 0 0>}
      localJointOffset=<0.75 0 0>
     }]
    simSpaceSettings {maxAngularAcceleration=25, maxAngularVelocity=5, simSpaceAngularAlpha=1}
   }
   n27 = Anim Dynamics - Bone: Feather_03_2(ComponentPose=@n26, Alpha=1, ExternalForce="0,0,0") {
    angularDampingOverride=1
    bGravityOverrideInSimSpace=true
    bOverrideAngularDamping=true
    bUseGravityOverride=true
    simulationSpace=RootRelative
    boundBone {boneName=Feather_03_2}
    componentAppliedLinearAccClamp=<5000 5000 5000>
    componentLinearAccScale=<0.5 0.5 0.5>
    componentLinearVelScale=<0.5 0.5 0.5>
    gravityOverride {z=980}
    physicsBodyDefinitions [{
      boundBone {boneName=Feather_03_2}
      boxExtents=<10 10 10>
      constraintSetup {angularLimitsMax=<3 15 0>, angularLimitsMin=<-15 -15 -15>, angularTarget=<1 0 0>}
      localJointOffset=<0.5 0 0>
     }]
    simSpaceSettings {maxAngularAcceleration=25, maxAngularVelocity=5, simSpaceAngularAlpha=1}
   }
   n28 = Anim Dynamics - Bone: Feather_03_3(ComponentPose=@n27, Alpha=1, ExternalForce="0,0,0") {
    angularDampingOverride=1
    bGravityOverrideInSimSpace=true
    bOverrideAngularDamping=true
    bUseGravityOverride=true
    simulationSpace=RootRelative
    boundBone {boneName=Feather_03_3}
    componentAppliedLinearAccClamp=<3000 3000 3000>
    componentLinearAccScale=<0.5 0.5 0.5>
    componentLinearVelScale=<0.5 0.5 0.5>
    gravityOverride {z=980}
    physicsBodyDefinitions [{
      boundBone {boneName=Feather_03_3}
      boxExtents=<10 10 10>
      constraintSetup {angularLimitsMax=<3 15 0>, angularLimitsMin=<-15 -15 -15>, angularTarget=<1 0 0>}
      localJointOffset=<0.5 0 0>
     }]
    simSpaceSettings {maxAngularAcceleration=25, maxAngularVelocity=5, simSpaceAngularAlpha=1}
   }
   n29 = Anim Dynamics - Bone: Feather_03_4(ComponentPose=@n28, Alpha=1, ExternalForce="0,0,0") {
    angularDampingOverride=0.999
    bGravityOverrideInSimSpace=true
    bOverrideAngularDamping=true
    bUseGravityOverride=true
    simulationSpace=RootRelative
    boundBone {boneName=Feather_03_4}
    componentAppliedLinearAccClamp=<3000 3000 3000>
    componentLinearAccScale=<0.5 0.5 0.5>
    componentLinearVelScale=<0.5 0.5 0.5>
    gravityOverride {z=980}
    physicsBodyDefinitions [{
      boundBone {boneName=Feather_03_4}
      boxExtents=<10 10 10>
      constraintSetup {angularLimitsMax=<3 15 10>, angularLimitsMin=<-15 -15 -10>, angularTarget=<1 0 0>}
      localJointOffset=<0.75 0 0>
     }]
    simSpaceSettings {maxAngularAcceleration=25, maxAngularVelocity=5, simSpaceAngularAlpha=1}
   }
   n30 = Anim Dynamics - Bone: Feather_04_1(ComponentPose=@n29, Alpha=1, ExternalForce="0,0,0") {
    angularDampingOverride=1
    bGravityOverrideInSimSpace=true
    bOverrideAngularDamping=true
    bUseGravityOverride=true
    simulationSpace=RootRelative
    alphaScaleBiasClamp {bMapRange=true}
    boundBone {boneName=Feather_04_1}
    componentAppliedLinearAccClamp=<5000 5000 5000>
    componentLinearAccScale=<0.5 0.5 0.5>
    componentLinearVelScale=<0.5 0.5 0.5>
    gravityOverride {z=980}
    physicsBodyDefinitions [{
      boundBone {boneName=Feather_04_1}
      boxExtents=<10 10 10>
      constraintSetup {angularLimitsMax=<5 5 0>, angularLimitsMin=<-5 -5 -5>, angularTarget=<1 0 0>}
      localJointOffset=<0.75 0 0>
     }]
    simSpaceSettings {maxAngularAcceleration=25, maxAngularVelocity=5, simSpaceAngularAlpha=1}
   }
   n31 = Anim Dynamics - Bone: Feather_04_2(ComponentPose=@n30, Alpha=1, ExternalForce="0,0,0") {
    angularDampingOverride=1
    bGravityOverrideInSimSpace=true
    bOverrideAngularDamping=true
    bUseGravityOverride=true
    simulationSpace=RootRelative
    boundBone {boneName=Feather_04_2}
    componentAppliedLinearAccClamp=<5000 5000 5000>
    componentLinearAccScale=<0.5 0.5 0.5>
    componentLinearVelScale=<0.5 0.5 0.5>
    gravityOverride {z=980}
    physicsBodyDefinitions [{
      boundBone {boneName=Feather_04_2}
      boxExtents=<10 10 10>
      constraintSetup {angularLimitsMax=<10 15 0>, angularLimitsMin=<-10 -15 -15>, angularTarget=<1 0 0>}
      localJointOffset=<0.5 0 0>
     }]
    simSpaceSettings {maxAngularAcceleration=25, maxAngularVelocity=5, simSpaceAngularAlpha=1}
   }
   n32 = Anim Dynamics - Bone: Feather_04_3(ComponentPose=@n31, Alpha=1, ExternalForce="0,0,0") {
    angularDampingOverride=1
    bGravityOverrideInSimSpace=true
    bOverrideAngularDamping=true
    bUseGravityOverride=true
    simulationSpace=RootRelative
    boundBone {boneName=Feather_04_3}
    componentAppliedLinearAccClamp=<3000 3000 3000>
    componentLinearAccScale=<0.5 0.5 0.5>
    componentLinearVelScale=<0.5 0.5 0.5>
    gravityOverride {z=980}
    physicsBodyDefinitions [{
      boundBone {boneName=Feather_04_3}
      boxExtents=<10 10 10>
      constraintSetup {angularLimitsMax=<10 15 0>, angularLimitsMin=<-10 -15 -15>, angularTarget=<1 0 0>}
      localJointOffset=<0.5 0 0>
     }]
    simSpaceSettings {maxAngularAcceleration=25, maxAngularVelocity=5, simSpaceAngularAlpha=1}
   }
   Output Pose(Result=Blend Poses by bool(BlendPose_0=Component To Local(ComponentPose=@n32), BlendPose_1=Component To Local(ComponentPose=@n16), bActiveValue=Get bIsInFrontEnd())) {layerGroup=None}
   AnimGraphNode_StateMachine_0 = Idle + Reactives() {is=title AnimGraphNode_StateMachine}
   AnimGraphNode_StateMachine_1 = FE_Reactivity() {is=title AnimGraphNode_StateMachine}
   n33 = Get bShouldDisplayFrontendReactivity() -> (bShouldDisplayFrontendReactivity:bool)
   n34 = "Transform (Modify) Bone - Bone: pack"(
    ComponentPose=Local To Component(LocalPose=Blend Poses by bool(BlendPose_0=@AnimGraphNode_StateMachine_1, BlendPose_1=@AnimGraphNode_StateMachine_0, bActiveValue=@n33) {blendType=Linear})
    Translation="-3,4,0"
    Rotation="0,0,0"
   ) {
    translationMode=BMM_Additive
    translationSpace=BCS_BoneSpace
    alphaBoolBlend {blendOption=Linear}
    boneToModify {boneName=pack}
    translation=<-3 4>
   }
   n35 = "Transform (Modify) Bone - Bone: pack"(ComponentPose=@n34, Translation="0.000,0.000,5.000", Rotation="0,0,0") {
    rotationMode=BMM_Additive
    translationMode=BMM_Additive
    translationSpace=BCS_BoneSpace
    alphaBoolBlend {blendOption=Linear}
    boneToModify {boneName=pack}
    translation {z=5}
   }
   n36 = "Transform (Modify) Bone - Bone: pack_anim"(ComponentPose=@n35, Alpha=1, Translation=Get Translate Main() -> (Translate Main:Vector), Rotation="0,0,0", Scale="1,1,1") {translationMode=BMM_Additive, translationSpace=BCS_BoneSpace, alphaBoolBlend {blendOption=Linear}, boneToModify {boneName=pack_anim}}
   n37 = "Transform (Modify) Bone - Bone: Feather_02_1"(ComponentPose=@n36, Alpha=1, Translation="0,0,0", Rotation="-0.002,-10.000,-0.025", Scale="1,1,1") {
    rotationMode=BMM_Additive
    translationMode=BMM_Additive
    translationSpace=BCS_BoneSpace
    alphaBoolBlend {blendOption=Linear}
    boneToModify {boneName=Feather_02_1}
    rotation=<rot -0.002 -10 -0.025>
   }
   n38 = "Transform (Modify) Bone - Bone: Feather_03_1"(ComponentPose=@n37, Alpha=1, Translation="0,0,0", Rotation="-0.604,-9.982,0.054", Scale="1,1,1") {
    rotationMode=BMM_Additive
    translationMode=BMM_Additive
    translationSpace=BCS_BoneSpace
    alphaBoolBlend {blendOption=Linear}
    boneToModify {boneName=Feather_03_1}
    rotation=<rot -0.604 -9.982 0.054>
   }
   n39 = "Transform (Modify) Bone - Bone: Feather_01_4"(ComponentPose=@n38, Alpha=1, Translation="0,0,0", Rotation="-4.809,7.580,9.182", Scale="1,1,1") {
    rotationMode=BMM_Additive
    translationMode=BMM_Additive
    translationSpace=BCS_BoneSpace
    alphaBoolBlend {blendOption=Linear}
    boneToModify {boneName=Feather_01_4}
    rotation=<rot -4.809 7.58 9.182>
   }
   Save cached pose 'PreDyn'(Pose=Component To Local(ComponentPose=@n39)) {cacheName=PreDyn}
  ]}
 Idle + Reactives {
  entry=Idle
  path=AnimGraph.AnimGraphNode_StateMachine_0.Idle + Reactives
  ids=id
  nodes [
   AnimStateNode_0 = Idle() {stateEntered {guid=#1, notifyName=Entered_RandomLocoBreak}}
   AnimStateNode_1 = TakeDamage() {is=title AnimStateNode}
   AnimStateNode_2 = Skydiving() {
    is=title AnimStateNode
    stateEntered {guid=#2}
    stateFullyBlended {guid=#3}
    stateLeft {guid=#4}
   }
   AnimStateNode_3 = Gain Health/Shield() {is=title AnimStateNode}
   AnimStateNode_4 = Elimination() {is=title AnimStateNode}
   AnimStateNode_5 = OpenChest() {is=title AnimStateNode}
   AnimStateTransitionNode_0 = AnimStateTransitionNode() {bAllowInertializationForSelfTransitions=true}
   AnimStateTransitionNode_1 = AnimStateTransitionNode() {bAllowInertializationForSelfTransitions=true, bAutomaticRuleBasedOnSequencePlayerInState=true}
   AnimStateTransitionNode_10 = AnimStateTransitionNode() {crossfadeDuration=0.3}
   AnimStateTransitionNode_11 = AnimStateTransitionNode()
   AnimStateTransitionNode_12 = AnimStateTransitionNode() {crossfadeDuration=0.05}
   AnimStateTransitionNode_13 = AnimStateTransitionNode() {bAutomaticRuleBasedOnSequencePlayerInState=true, crossfadeDuration=0.05}
   AnimStateTransitionNode_2 = AnimStateTransitionNode() {crossfadeDuration=0.3}
   AnimStateTransitionNode_5 = AnimStateTransitionNode() {bAutomaticRuleBasedOnSequencePlayerInState=true, crossfadeDuration=0}
   AnimStateTransitionNode_7 = AnimStateTransitionNode()
   AnimStateTransitionNode_8 = AnimStateTransitionNode() {bAutomaticRuleBasedOnSequencePlayerInState=true}
  ]
  links [
   {
    dst=Idle
    rule=AnimGraph.AnimGraphNode_StateMachine_0.Idle + Reactives.AnimStateTransitionNode_5.Transition
    src=Elimination
    via=AnimStateTransitionNode_5
   }
   {
    dst=Idle
    rule=AnimGraph.AnimGraphNode_StateMachine_0.Idle + Reactives.AnimStateTransitionNode_1.Transition
    src=Gain Health/Shield
    via=AnimStateTransitionNode_1
   }
   {
    dst=Elimination
    rule=AnimGraph.AnimGraphNode_StateMachine_0.Idle + Reactives.AnimStateTransitionNode_11.Transition
    src=Idle
    via=AnimStateTransitionNode_11
   }
   {
    dst=Gain Health/Shield
    rule=AnimGraph.AnimGraphNode_StateMachine_0.Idle + Reactives.AnimStateTransitionNode_0.Transition
    src=Idle
    via=AnimStateTransitionNode_0
   }
   {
    dst=OpenChest
    rule=AnimGraph.AnimGraphNode_StateMachine_0.Idle + Reactives.AnimStateTransitionNode_7.Transition
    src=Idle
    via=AnimStateTransitionNode_7
   }
   {
    dst=Skydiving
    rule=AnimGraph.AnimGraphNode_StateMachine_0.Idle + Reactives.AnimStateTransitionNode_10.Transition
    src=Idle
    via=AnimStateTransitionNode_10
   }
   {
    dst=TakeDamage
    rule=AnimGraph.AnimGraphNode_StateMachine_0.Idle + Reactives.AnimStateTransitionNode_12.AnimationTransitionGraph_0
    src=Idle
    via=AnimStateTransitionNode_12
   }
   {
    dst=Idle
    rule=AnimGraph.AnimGraphNode_StateMachine_0.Idle + Reactives.AnimStateTransitionNode_8.Transition
    src=OpenChest
    via=AnimStateTransitionNode_8
   }
   {
    dst=Idle
    rule=AnimGraph.AnimGraphNode_StateMachine_0.Idle + Reactives.AnimStateTransitionNode_2.Transition
    src=Skydiving
    via=AnimStateTransitionNode_2
   }
   {
    dst=Idle
    rule=AnimGraph.AnimGraphNode_StateMachine_0.Idle + Reactives.AnimStateTransitionNode_13.AnimationTransitionGraph_0
    src=TakeDamage
    via=AnimStateTransitionNode_13
   }
  ]
 }
 Idle {
  path=AnimGraph.AnimGraphNode_StateMachine_0.Idle + Reactives.AnimStateNode_0.Idle
  nodes [Output Animation Pose(Result=Sequence Player 'SpruceBark_Idle'() {sequence=/BRCosmetics/Accessories/FORT_Backpacks/Backpack_M_MED_SpruceBark/Animations/SpruceBark_Idle})]
 }
 TakeDamage {
  path=AnimGraph.AnimGraphNode_StateMachine_0.Idle + Reactives.AnimStateNode_1.TakeDamage
  nodes [Output Animation Pose(Result=Sequence Player 'SpruceBark_TakeDamage'() {bLoopAnimation=false, sequence=/BRCosmetics/Accessories/FORT_Backpacks/Backpack_M_MED_SpruceBark/Animations/SpruceBark_TakeDamage})]
 }
 Skydiving {
  path=AnimGraph.AnimGraphNode_StateMachine_0.Idle + Reactives.AnimStateNode_2.Skydiving
  nodes [Output Animation Pose(Result=Sequence Player 'SpruceBark_Flying'() {sequence=/BRCosmetics/Accessories/FORT_Backpacks/Backpack_M_MED_SpruceBark/Animations/SpruceBark_Flying})]
 }
 Gain Health/Shield {
  path=AnimGraph.AnimGraphNode_StateMachine_0.Idle + Reactives.AnimStateNode_3.Gain Health/Shield
  nodes [Output Animation Pose(Result=Sequence Player 'SpruceBark_GainHealth'() {bLoopAnimation=false, sequence=/BRCosmetics/Accessories/FORT_Backpacks/Backpack_M_MED_SpruceBark/Animations/SpruceBark_GainHealth})]
 }
 Elimination {
  path=AnimGraph.AnimGraphNode_StateMachine_0.Idle + Reactives.AnimStateNode_4.Elimination
  nodes [Output Animation Pose(Result=Sequence Player 'SpruceBark_Elim'() {bLoopAnimation=false, sequence=/BRCosmetics/Accessories/FORT_Backpacks/Backpack_M_MED_SpruceBark/Animations/SpruceBark_Elim}) {
    node {stateExitFunction {functionName=OnStateExit}}
    stateExitFunction {bSelfContext=true, memberGuid=#5, memberName=OnStateExit}
   }]
 }
 OpenChest {
  path=AnimGraph.AnimGraphNode_StateMachine_0.Idle + Reactives.AnimStateNode_5.OpenChest
  nodes [Output Animation Pose(Result=Sequence Player 'SpruceBark_OpenChest'() {bLoopAnimation=false, sequence=/BRCosmetics/Accessories/FORT_Backpacks/Backpack_M_MED_SpruceBark/Animations/SpruceBark_OpenChest})]
 }
 Transition {
  path=AnimGraph.AnimGraphNode_StateMachine_0.Idle + Reactives.AnimStateTransitionNode_0.Transition
  nodes [Result(bCanEnterTransition=Get GainedHealth() -> (GainedHealth:bool))]
 }
 Transition {path=AnimGraph.AnimGraphNode_StateMachine_0.Idle + Reactives.AnimStateTransitionNode_1.Transition, nodes [Result()]}
 Transition {
  path=AnimGraph.AnimGraphNode_StateMachine_0.Idle + Reactives.AnimStateTransitionNode_10.Transition
  nodes [Result(bCanEnterTransition=Get bIsSkyDiving())]
 }
 Transition {
  path=AnimGraph.AnimGraphNode_StateMachine_0.Idle + Reactives.AnimStateTransitionNode_11.Transition
  nodes [Result(bCanEnterTransition=Get bEnemyPlayerEliminated() -> (bEnemyPlayerEliminated:bool))]
 }
 AnimationTransitionGraph_0 {
  path=AnimGraph.AnimGraphNode_StateMachine_0.Idle + Reactives.AnimStateTransitionNode_12.AnimationTransitionGraph_0
  nodes [Result(bCanEnterTransition=Get TookDamage() -> (TookDamage:bool))]
 }
 AnimationTransitionGraph_0 {path=AnimGraph.AnimGraphNode_StateMachine_0.Idle + Reactives.AnimStateTransitionNode_13.AnimationTransitionGraph_0, nodes [Result()]}
 Transition {
  path=AnimGraph.AnimGraphNode_StateMachine_0.Idle + Reactives.AnimStateTransitionNode_2.Transition
  nodes [Result(bCanEnterTransition=NOT Boolean(A=Get bIsSkyDiving()))]
 }
 Transition {path=AnimGraph.AnimGraphNode_StateMachine_0.Idle + Reactives.AnimStateTransitionNode_5.Transition, nodes [Result()]}
 Transition {
  path=AnimGraph.AnimGraphNode_StateMachine_0.Idle + Reactives.AnimStateTransitionNode_7.Transition
  nodes [Result(bCanEnterTransition=Get bChestOpened() -> (bChestOpened:bool))]
 }
 Transition {path=AnimGraph.AnimGraphNode_StateMachine_0.Idle + Reactives.AnimStateTransitionNode_8.Transition, nodes [Result()]}
 FE_Reactivity {
  entry=Idle
  path=AnimGraph.AnimGraphNode_StateMachine_1.FE_Reactivity
  ids=id
  nodes [
   AnimStateNode_0 = Idle()
   AnimStateNode_2 = FE_ChestOpen() {is=title AnimStateNode, stateEntered {guid=#6, notifyName=Entered FE_Reactive_Anim}}
   AnimStateTransitionNode_8 = AnimStateTransitionNode()
   AnimStateTransitionNode_9 = AnimStateTransitionNode() {bAutomaticRuleBasedOnSequencePlayerInState=true}
  ]
  links [
   {
    dst=Idle
    rule=AnimGraph.AnimGraphNode_StateMachine_1.FE_Reactivity.AnimStateTransitionNode_9.Transition
    src=FE_ChestOpen
    via=AnimStateTransitionNode_9
   }
   {
    dst=FE_ChestOpen
    rule=AnimGraph.AnimGraphNode_StateMachine_1.FE_Reactivity.AnimStateTransitionNode_8.Transition
    src=Idle
    via=AnimStateTransitionNode_8
   }
  ]
 }
 Idle {
  path=AnimGraph.AnimGraphNode_StateMachine_1.FE_Reactivity.AnimStateNode_0.Idle
  nodes [Output Animation Pose(Result=Sequence Player 'SpruceBark_Idle'() {sequence=&SpruceBark_Idle})]
 }
 FE_ChestOpen {
  path=AnimGraph.AnimGraphNode_StateMachine_1.FE_Reactivity.AnimStateNode_2.FE_ChestOpen
  nodes [Output Animation Pose(Result=Sequence Player 'SpruceBark_OpenChest'() {bLoopAnimation=false, sequence=&SpruceBark_OpenChest})]
 }
 Transition {
  path=AnimGraph.AnimGraphNode_StateMachine_1.FE_Reactivity.AnimStateTransitionNode_8.Transition
  nodes [Result(bCanEnterTransition=Get Play Reactive FE Anim() -> (Play Reactive FE Anim:bool))]
 }
 Transition {path=AnimGraph.AnimGraphNode_StateMachine_1.FE_Reactivity.AnimStateTransitionNode_9.Transition, nodes [Result()]}
 EventGraph {nodes [
   n1 = Cast To FortPlayerPawn(Object=GetOwningActor())
   n2 = GetHealth!(self:Fort Pawn Object Reference=@n1) -> (ReturnValue:float)
   > Set Current Health(Current Health:double=@n2) -> (Output_Get:double)
   > n3 = Branch(Condition="Not Equal (Float)"(A=@n2, B=Get Previous Health() -> (Previous Health:double)))
   > Set TookDamage(TookDamage:bool=true) -> (Output_Get:bool)
   > Set Previous Health(Previous Health:double=Get Current Health() -> (Current Health:double)) -> (Output_Get:double)
   Set TookDamage(execute=@n3.else, TookDamage=false)
   Event OnCharacterCustomizationLoaded()
   > Cast To FortPlayerPawn(Object=TryGetPawnOwner())
   > n4 = Branch()
   > Do Once(Reset=@n6.then)
   > n5 = Delay(Duration=4)
   Set Play Reactive FE Anim(execute=[@n7.then, @n5.then], Play Reactive FE Anim:bool=true) -> (Output_Get:bool)
   n6 = Set Play Reactive FE Anim(execute=@n4.else, Play Reactive FE Anim=false)
   AnimNotify_FE_ReactiveAnim() {is=title K2Node_Event}
   > Set Play Reactive FE Anim(Play Reactive FE Anim=false)
   > n7 = Delay(Duration=10)
   n8 = Event BlueprintUpdateAnimation()
   > n9 = Sequence!(execute) -> (then_0, then_1, then_2:Exec, then_3:Exec)
   n10 = GetRealTimeSeconds()
   n11 = "Sin (Radians)"(A=float * float(A=float * float(A=1, B=-1.4), B=@n10) {is=title K2Node_PromotableOperator})
   n12 = "Sin (Radians)"(A=float * float(A=float * float(A=0.8, B=-0.8), B=@n10))
   > Set Translate Main(Translate Main:Vector=MakeVector(X=float / float(A=@n12, B=-2), Y=0, Z=float / float(A=@n11, B=-1))) -> (Output_Get:Vector)
   n13 = Set AnimDynamicsEnabled(
    execute=@n9.then_1
    AnimDynamicsEnabled:double=Remap Float(Input=GetCurveValue(CurveName=IsPhysicsDisabled), OldRange=(X=0,Y=1), NewRange=(X=1,Y=0), Clamp=true) -> (Output=~)
   ) -> (Output_Get:double)
   > n14 = Branch(Condition=OR Boolean(A="float <= float"(A=@n13) -> (ReturnValue=~), B=Get bEnemyPlayerEliminated()))
   > Set AnimDyn_LR1(AnimDyn_LR1:double=@n13) -> (Output_Get:double)
   > Set AnimDyn_LR2(AnimDyn_LR2:double=@n13) -> (Output_Get:double)
   > Set AnimDyn_LR3(AnimDyn_LR3:double=@n13) -> (Output_Get:double)
   Set AnimDyn_LR1(execute=@n14.else, AnimDyn_LR1=Remap Float(Input=@n13, OldRange=(X=0,Y=1), NewRange=(X=0.25,Y=1), Clamp=true) -> (Output=~))
   > Set AnimDyn_LR2(AnimDyn_LR2=Remap Float(Input=@n13, OldRange=(X=0,Y=1), NewRange=(X=0.5,Y=1), Clamp=true) -> (Output=~))
   > Set AnimDyn_LR3(AnimDyn_LR3=Remap Float(Input=@n13, OldRange=(X=0,Y=1), NewRange=(X=0.75,Y=1), Clamp=true) -> (Output=~))
   n15 = GetOwnerSkeletalMeshComponent()
   n16 = BreakTransform(InTransform=GetDeltaTransformFromRefPose(self=@n15, BoneName=spine_05, BaseName=pelvis))
   n17 = BreakRotator(InRot=@n16.Rotation)
   n18 = BreakRotator(InRot=BreakTransform(InTransform=GetDeltaTransformFromRefPose(self=@n15, BoneName=head, BaseName=spine_05)).Rotation)
   n19 = BreakRotator(InRot=BreakTransform(InTransform=GetDeltaTransformFromRefPose(self=@n15, BoneName=head, BaseName=pelvis)).Rotation)
   n20 = float * float(A=BreakVector(InVec=@n16).Y, B:bool=Get bIsSkyDiving())
   n21 = BreakTransform(InTransform=GetDeltaTransformFromRefPose(self=@n15, BoneName=spine_04, BaseName=spine_03))
   n22 = float + float!(
    A=MapRangeClamped(Value=@n17.Yaw, InRangeA=-8, InRangeB=-30, OutRangeB=15)
    B=MapRangeClamped(Value=@n17.Yaw, InRangeA=-30, InRangeB=-81, OutRangeB=50)
    C:double=MapRangeClamped(Value=@n18.Yaw, InRangeA=-60, InRangeB=-102, OutRangeB=-25)
    D:double=MapRangeClamped(Value=@n19.Yaw, InRangeA=-60, InRangeB=-150, OutRangeB=35)
    E:double=MapRangeClamped(Value=@n20, InRangeA=0, InRangeB=10, OutRangeB=40)
    F:double=MapRangeClamped(Value=BreakRotator(InRot=@n21.Rotation).Yaw, InRangeA=-20, InRangeB=-40, OutRangeB=50)
   ) -> (ReturnValue)
   n23 = Get!() -> (Backpack Rotate_Roll:float, Backpack Rotate_Pitch:float, Backpack Rotate_Yaw:float)
   Set Backpack Rotate(execute=@n9.then_2, Backpack Rotate:Rotator=MakeRotator(Roll=FInterpTo(Current=@n23, Target=@n22, DeltaTime=@n8.DeltaTimeX, InterpSpeed=3))) -> (Output_Get:Rotator)
   n24 = float + float!(
    A=MapRangeClamped(Value=@n18.Yaw, InRangeA=-60, InRangeB=-102, OutRangeB=-10)
    B=MapRangeClamped(Value=@n19.Yaw, InRangeA=-80, InRangeB=-102, OutRangeB=-10)
    C=MapRangeClamped(Value=@n20, InRangeA=0, InRangeB=10, OutRangeB=10)
   ) -> (ReturnValue)
   n25 = Get!() -> (Backpack Translation_X:double, Backpack Translation_Y:double, Backpack Translation_Z:double)
   n26 = float + float!(
    A=MapRangeClamped(Value=@n17.Yaw, InRangeA=-8, InRangeB=-30, OutRangeB=-3)
    B=MapRangeClamped(Value=@n19.Yaw, InRangeA=-60, InRangeB=-102, OutRangeB=-10)
   ) -> (ReturnValue)
   > Set Backpack Translation(Backpack Translation:Vector=MakeVector(
     Y=FInterpTo(Current=@n25.Backpack Translation_Y, Target=@n26, DeltaTime=@n8.DeltaTimeX, InterpSpeed=3)
     Z=FInterpTo(Current=@n25.Backpack Translation_Z, Target=@n24, DeltaTime=@n8.DeltaTimeX, InterpSpeed=3)
    )) -> (Output_Get:Vector)
   n27 = Event BlueprintPostEvaluateAnimation() {is=title K2Node_Event}
   Set bChestOpened(execute=[@n27.then, @n31.then], bChestOpened:bool=false) -> (Output_Get:bool)
   > Set GainedHealth(GainedHealth:bool=false) -> (Output_Get:bool)
   > Set Play Reactive FE Anim(Play Reactive FE Anim=false)
   AnimNotify_ResetReactiveAnimation() {is=title K2Node_Event}
   > Set GainedHealth(GainedHealth=false)
   n28 = Event BlueprintProcessAnimInputEvent()
   > n29 = Sequence!(execute) -> (then_0, then_1, then_2)
   > n30 = Branch(Condition=IsNamed(self=@n28.AnimInputEvent, InEventName=aie.EnemyPlayerEliminated))
   > Set bEnemyPlayerEliminated(bEnemyPlayerEliminated:bool=true) -> (Output_Get:bool)
   Branch(execute=@n30.else, Condition=IsNamed(self=@n28.AnimInputEvent, InEventName=aie.ChestOpened))
   > Set bChestOpened(bChestOpened=true)
   Branch(
    execute=@n29.then_1
    Condition=OR Boolean(
     A=IsNamed(self=@n28.AnimInputEvent, InEventName=aie.PlayerVitality.HealthDecreased)
     B=IsNamed(self=@n28.AnimInputEvent, InEventName=Noneaie.PlayerVitality.ShieldsDecreased)
    )
   )
   > Set TookDamage(TookDamage=true)
   > RetriggerableDelay(Duration=0.2)
   > Set TookDamage(TookDamage=false)
   Branch(
    execute=@n29.then_2
    Condition=OR Boolean(
     A=IsNamed(self=@n28.AnimInputEvent, InEventName=aie.PlayerVitality.ShieldsIncreased)
     B=IsNamed(self=@n28.AnimInputEvent, InEventName=aie.PlayerVitality.HealthIncreased)
    )
   )
   > Set GainedHealth(GainedHealth=true)
   > RetriggerableDelay(Duration=0.2)
   > Set GainedHealth(GainedHealth=false)
   n31 = Set bEnemyPlayerEliminated(bEnemyPlayerEliminated=false)
   n32 = GetOwnerSkeletalMeshComponent()
   n33 = Set head down rot(head down rot:double=GetAlphaFromReferenceTransformDelta(SkeletalMeshComponent=@n32, BoneName=spine_01, BaseName=head, InputRange=(X=0,Y=-45), TwistAxis="1,0,0")) -> (Output_Get:double)
   n34 = "To Float (Boolean)"(InBool=Get bIsPlayingMeleeAnim())
   n35 = float * float(B:double=0.2)
   n36 = Get isMelee() -> (isMelee:double)
   > Set isMelee(isMelee:double=float * float(A=FInterpTo(Current=@n36, Target=@n34, DeltaTime=@n35, InterpSpeed=5), B=@n33)) -> (Output_Get:double)
   n37 = BreakRotator(InRot=BreakTransform(InTransform=GetDeltaTransformFromRefPose(self=@n32, BoneName=upperarm_l, BaseName=clavicle_l)).Rotation)
   > Set Backpack Rotate Pickaxe(Backpack Rotate Pickaxe:Rotator=MakeRotator(Roll=MapRangeClamped(Value=@n37.Yaw, InRangeA=0, InRangeB=90, OutRangeB=30))) -> (Output_Get:Rotator)
   TryGetPawnOwner()
   Reroute Node(InputPin:Wildcard) -> (OutputPin:Wildcard)
   n38 = float * float(A=1, B=1.3)
   float / float(A="Sin (Radians)"(A=float * float(A=@n38, B=@n10)), B=20)
  ]}
 OnStateExit {nodes [
   OnStateExit!() -> (then, Context:"Anim Update Context Structure (by ref)", Node:"Anim Node Reference Structure (by ref)") {is=title K2Node_FunctionEntry}
   > Set bEnemyPlayerEliminated(bEnemyPlayerEliminated=false)
   > Return Node()
  ]}
]
variables [
 AllowLocomotionBreak {type=boolean}
 AnimDyn_LR1 {type=number}
 AnimDyn_LR2 {type=number}
 AnimDyn_LR3 {type=number}
 AnimDynamicsEnabled {type=number, default=1}
 As Player Pawn Athena {type=PlayerPawn_Athena_C}
 Backpack Rotate {type=Rotator}
 Backpack Rotate Pickaxe {type=Rotator}
 Backpack Translation {type=Vector}
 BackPackRotate {type=Rotator}
 BackPackTranslate {type=Vector}
 bChestOpened {type=boolean}
 bEnemyPlayerEliminated {type=boolean}
 Choose_LocoBreak_Index {type=integer}
 Current Health {type=number}
 FrontEnd {type=boolean}
 GainedHealth {type=boolean}
 head down rot {type=number}
 HealthDecreased {type=boolean}
 HealthIncreased {type=boolean}
 IsLocoBreaking {type=boolean}
 isMelee {type=number}
 Play Reactive FE Anim {type=boolean}
 Previous Health {type=number}
 Reactivity Enabled {type=boolean}
 ShieldsDecreased {type=boolean}
 ShieldsIncreased {type=boolean}
 Skydive {type=number}
 TookDamage {type=boolean}
 Translate Main {type=Vector}
]
